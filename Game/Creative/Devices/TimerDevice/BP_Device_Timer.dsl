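; Creative timer device that runs per-player or shared countdown timers in minigames, filtered by team and class, and can award score.
; Channel receivers drive the timer state machine for one or all tracked players, with shared timers keyed by a null controller.
; Minigame events manage tracked timers, while persistence saves and loads personal timers, optionally automatically.
; Clients rebuild text and HUD displays from server finish time and play urgency glow and audio.
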(event EventOnWorldReady
  ;[0 1 0.09771 0.25] Initialization
  (Visibility|SetVisibilityComponents (Variables|Components|GetVisibleInGameComponent) (Utilities|Array|MakeArray (Variables|Components|GetTimerDisplayTextFront) (Variables|Components|GetTimerDisplayTextBack) (Variables|Default|GetUrgencyGlow)) true)
  (Visibility|SetVisibilityComponents (Variables|Components|GetMeshVisibleInEditorComponent) (Utilities|Array|MakeArray (Variables|Debug|GetStaticMeshComponent)) true))

(event OnAnyPropertyChanged(ToyOptionsComponent)
  ;[0 1 0.09771 0.25] Initialization
  (Delegates|BindDelegates)
  (CallFunction|BindClientExitVolumeGameplayMessage)
  (Initialization|InitializeComponent (Variables|Default|GetCreative_UseRestriction_Component) (Variables|Components|GetFortMinigameProgress) 0 0 (Variables|UserOptions|GetTeamSetting) (Variables|UserOptions|GetClassIndex) 0 1 :!class /Game/Creative/Devices/Common/Components/Creative_UseRestriction_Component.Creative_UseRestriction_Component_C)
  (Variables|Default|SetDeviceTeamID (Teams|IndextoTeam (Knobs|Team|GetTeamValueasIndex (Variables|UserOptions|GetTeamSetting))))
  (Visual|UpdateinGameVisibilities)
  (Class|CreativeGenericTextWidgetComp|UpdateText (Variables|Default|GetTimerWidget) (Utilities|Text|ToText(String) (Variables|UserOptions|GetTimerNameKnob)))
  (Visual|UpdateTimerColor (Variables|UserOptions|GetTimerColorKnob))
  (Variables|FortCreativeTimerDevice|SetClientTimerTextStyle (Variables|UserOptions|GetTimerTextStyleKnob))
  (Variables|FortCreativeTimerDevice|SetbClientShowOnHud (Math|Conversions|ToBoolean(Integer) (Variables|UserOptions|GetShowOnHUDKnob)))
  (Class|CreativeUIBroadcastComponent|Initialize (Variables|Default|GetCreative_HUDBroadcast_Component) (Variables|Components|GetFortMinigameProgress) (Variables|UserOptions|Advanced|GetDisplayScoreUpdateOnHUD)))

(event OnPropertiesLoaded(ToyOptionsComponent)
  ;[0 1 0.09771 0.25] Initialization
  (Delegates|BindDelegates)
  (CallFunction|BindClientExitVolumeGameplayMessage)
  (Initialization|InitializeComponent (Variables|Default|GetCreative_UseRestriction_Component) (Variables|Components|GetFortMinigameProgress) 0 0 (Variables|UserOptions|GetTeamSetting) (Variables|UserOptions|GetClassIndex) 0 1 :!class /Game/Creative/Devices/Common/Components/Creative_UseRestriction_Component.Creative_UseRestriction_Component_C)
  (Variables|Default|SetDeviceTeamID (Teams|IndextoTeam (Knobs|Team|GetTeamValueasIndex (Variables|UserOptions|GetTeamSetting))))
  (Visual|UpdateinGameVisibilities)
  (Class|CreativeGenericTextWidgetComp|UpdateText (Variables|Default|GetTimerWidget) (Utilities|Text|ToText(String) (Variables|UserOptions|GetTimerNameKnob)))
  (Visual|UpdateTimerColor (Variables|UserOptions|GetTimerColorKnob))
  (Variables|FortCreativeTimerDevice|SetClientTimerTextStyle (Variables|UserOptions|GetTimerTextStyleKnob))
  (Variables|FortCreativeTimerDevice|SetbClientShowOnHud (Math|Conversions|ToBoolean(Integer) (Variables|UserOptions|GetShowOnHUDKnob)))
  (Class|CreativeUIBroadcastComponent|Initialize (Variables|Default|GetCreative_HUDBroadcast_Component) (Variables|Components|GetFortMinigameProgress) (Variables|UserOptions|Advanced|GetDisplayScoreUpdateOnHUD)))

(event EventBeginPlay
  ;[0 1 0.09771 0.25] Initialization
  (|CallParentFunction)
  (Networking|SwitchHasAuthority
    (:Authority
      (Persistence|InitializePersistenceComponent (Variables|Default|GetFortCreativePersistenceComponentBase) (Variables|Components|GetFortMinigameProgress) true)
      (bind _returnvalue (Game|ConstructObjectfromClass (Variables|Persistence|GetPersistenceData_Timer_Class) self))
      (Variables|Persistence|SetPersistenceData_Timer _returnvalue))))

(event EditorPreview|EventSetupForEditorPreview
  ;[0 1 0.09771 0.25] Initialization
  ;;[0 0.5156 0.1638 0] Editor Only - When moving, we need to reconstruct the timer widget every frame as the references will get disconnected. This gets called every frame and does that.
  (|CallParentFunction)
  (Class|CreativeGenericTextWidgetComp|UpdateText (Variables|Default|GetTimerWidget) (Utilities|Text|ToText(String) (Variables|UserOptions|GetTimerNameKnob))))

(event EventEndPlay
  (params (EndPlayReason :type EEndPlayReason%20Enum :default "Destroyed"))
  ;[1 0.08089 0 0.25] Destruction Cleanup
  (|CallParentFunction EndPlayReason)
  (Delegates|UnbindDelegates))

(event EventClearDisplay
  ;[1 1 1] Client side reconcilation from server
  (Networking|CanExecuteCosmeticEvents
    (:True
      (Visual|ClearDisplayTexts)
      (Variables|Cosmetic|SetbUrgencyHasStarted false))))

(event Custom|StartTimerFromEvent
  (params (Instigator :type Controller%20Object%20Reference))
  ;[0.3706 0 1 0.25] Channel Receivers
  (bind _output_get (Variables|Default|SetLastInstigator Instigator))
  (Utilities|IsValid _output_get
    (:Is%20Valid
      (Variables|Default|SetActivatingPlayer _output_get))
    (:Is%20Not%20Valid
      _))
  (bind (_bexists _trackedplayer) (Default|GetIntendedTrackedPlayer _output_get))
  (if _bexists
    (switch Utilities|FlowControl|Switch|SwitchonETimerDeviceState (CallFunction|GetTimerState _trackedplayer)
      (:Enabled
        (CallFunction|TimerSetState _trackedplayer "Activated")))))

(event EventDisplayUpdate
  (params (CurrentState :type Byte) (ServerTimeFinished :type Float%20%28single-precision%29))
  (bind _returnvalue (Fortnite|Sync|Time|GetServerWorldTimeSeconds))
  (bind _returnvalue_2 (- ServerTimeFinished _returnvalue))
  (bind _returnvalue_3 (Utilities|Enum|BytetoEnumETimerDeviceState CurrentState))
  ;[1 1 1] Client side reconcilation from server
  (Networking|CanExecuteCosmeticEvents
    (:True
      (Visual|UpdateinGameDisplay _returnvalue_3 _returnvalue_2)
      (Default|SetupClientDisplayUpdateTimer _returnvalue_3 _returnvalue_2))))

(event Custom|CompleteTimerFromEvent
  (params (Instigator :type Controller%20Object%20Reference))
  ;[0.3706 0 1 0.25] Channel Receivers
  (bind _output_get (Variables|Default|SetLastInstigator Instigator))
  (bind (_bexists _trackedplayer) (Default|GetIntendedTrackedPlayer _output_get))
  (if _bexists
    (switch Utilities|FlowControl|Switch|SwitchonETimerDeviceState (CallFunction|GetTimerState _trackedplayer)
      (:Enabled
        _)
      (:Activated
        _)
      (:Paused
        _))
    (CallFunction|TimerSetState _trackedplayer "Completed")))

(event Custom|ResetTimerFromEvent
  (params (Instigator :type Controller%20Object%20Reference))
  ;[0.3706 0 1 0.25] Channel Receivers
  (bind _output_get (Variables|Default|SetLastInstigator Instigator))
  (bind (_bexists _trackedplayer) (Default|GetIntendedTrackedPlayer _output_get))
  (if _bexists
    (CallFunction|ResetTimer _trackedplayer)))

(event Custom|CompletionStateChangeDelay
  (params (TrackedPlayer :type Controller%20Object%20Reference) (NewState :type ETimerDeviceState%20Enum :default "Enabled"))
  (CallFunction|TimerClearHandles TrackedPlayer)
  (Utilities|FlowControl|Delay (Variables|Default|GetCompletionDelay))
  (CallFunction|TimerSetState TrackedPlayer NewState))

(event Custom|PauseTimerFromEvent
  (params (Instigator :type Controller%20Object%20Reference))
  ;[0.3706 0 1 0.25] Channel Receivers
  (bind (_bexists _trackedplayer) (Default|GetIntendedTrackedPlayer Instigator))
  (if _bexists
    (switch Utilities|FlowControl|Switch|SwitchonETimerDeviceState (CallFunction|GetTimerState _trackedplayer)
      (:Activated
        (CallFunction|TimerSetState _trackedplayer "Paused")))))

(event Custom|ResumeTimerFromEvent
  (params (Instigator :type Controller%20Object%20Reference))
  ;[0.3706 0 1 0.25] Channel Receivers
  (bind _output_get (Variables|Default|SetLastInstigator Instigator))
  (bind (_bexists _trackedplayer) (Default|GetIntendedTrackedPlayer _output_get))
  (if _bexists
    (switch Utilities|FlowControl|Switch|SwitchonETimerDeviceState (CallFunction|GetTimerState _trackedplayer)
      (:Paused
        (CallFunction|TimerSetState _trackedplayer "Activated")))))

(event Custom|EnableTimerFromEvent
  (params (Instigator :type Controller%20Object%20Reference))
  ;[0.3706 0 1 0.25] Channel Receivers
  (bind (_bexists _trackedplayer) (Default|GetIntendedTrackedPlayer Instigator))
  (if _bexists
    (switch Utilities|FlowControl|Switch|SwitchonETimerDeviceState (CallFunction|GetTimerState _trackedplayer)
      (:Disabled
        (CallFunction|TimerSetState _trackedplayer)))))

(event Custom|DisableTimerFromEvent
  (params (Instigator :type Controller%20Object%20Reference))
  ;[0.3706 0 1 0.25] Channel Receivers
  (bind (_bexists _trackedplayer) (Default|GetIntendedTrackedPlayer Instigator))
  (if _bexists
    (CallFunction|DisableTimer _trackedplayer)))

(event Custom|StartAllTimersFromEvent
  (params (Instigator :type Controller%20Object%20Reference))
  ;[0.01124 1 0.2021] Channel Receivers for triggering all personal timers
  (bind _bresult (Default|CanTriggerAllPersonalTimers Instigator))
  (if _bresult
    (for _array_element (CallFunction|GetTrackedPlayers)
      (switch Utilities|FlowControl|Switch|SwitchonETimerDeviceState (CallFunction|GetTimerState _array_element)
        (:Enabled
          (CallFunction|TimerSetState _array_element "Activated"))))))

(event Custom|PauseAllTimersFromEvent
  (params (Instigator :type Controller%20Object%20Reference))
  ;[0.01124 1 0.2021] Channel Receivers for triggering all personal timers
  (bind _bresult (Default|CanTriggerAllPersonalTimers Instigator))
  (if _bresult
    (for _array_element (CallFunction|GetTrackedPlayers)
      (switch Utilities|FlowControl|Switch|SwitchonETimerDeviceState (CallFunction|GetTimerState _array_element)
        (:Activated
          (CallFunction|TimerSetState _array_element "Paused"))))))

(event Custom|ResumeAllTimersFromEvent
  (params (Instigator :type Controller%20Object%20Reference))
  ;[0.01124 1 0.2021] Channel Receivers for triggering all personal timers
  (bind _bresult (Default|CanTriggerAllPersonalTimers Instigator))
  (if _bresult
    (for _array_element (CallFunction|GetTrackedPlayers)
      (switch Utilities|FlowControl|Switch|SwitchonETimerDeviceState (CallFunction|GetTimerState _array_element)
        (:Paused
          (CallFunction|TimerSetState _array_element "Activated"))))))

(event Custom|CompleteAllTimersFromEvent
  (params (Instigator :type Controller%20Object%20Reference))
  ;[0.01124 1 0.2021] Channel Receivers for triggering all personal timers
  (bind _bresult (Default|CanTriggerAllPersonalTimers Instigator))
  (if _bresult
    (for _array_element (CallFunction|GetTrackedPlayers)
      (switch Utilities|FlowControl|Switch|SwitchonETimerDeviceState (CallFunction|GetTimerState _array_element)
        (:Enabled
          _)
        (:Activated
          _)
        (:Paused
          _))
      (CallFunction|TimerSetState _array_element "Completed"))))

(event Custom|ResetAllTimersFromEvent
  (params (Instigator :type Controller%20Object%20Reference))
  ;[0.01124 1 0.2021] Channel Receivers for triggering all personal timers
  (bind _bresult (Default|CanTriggerAllPersonalTimers Instigator))
  (if _bresult
    (for _array_element (CallFunction|GetTrackedPlayers)
      (CallFunction|ResetTimer _array_element))))

(event Custom|SaveTimerFromEvent
  (params (Instigator :type Controller%20Object%20Reference))
  (bind (_outstate _returnvalue) (CallFunction|GetTimerState Instigator))
  ;[0.09255 1 0.3687] Channel Receivers for Save/Load
  (Persistence|SaveTimer Instigator)
  (if (Math|Boolean|ANDBoolean (Utilities|Enum|Equal(Enum) (Variables|UserOptions|GetSaveBehaviorKnob) "NewEnumerator1") _returnvalue (Utilities|Enum|Equal(Enum) _outstate "Activated"))
    (CallFunction|TimerSetState Instigator "Paused")))

(event Custom|LoadTimerFromEvent
  (params (Instigator :type Controller%20Object%20Reference))
  ;[0.09255 1 0.3687] Channel Receivers for Save/Load
  (Persistence|LoadTimer Instigator))

(event Custom|ClearPersistenceDataFromEvent
  (params (Instigator :type Controller%20Object%20Reference))
  ;[0.09255 1 0.3687] Channel Receivers for Save/Load
  (bind _bresult (Default|CanTriggerAllPersonalTimers Instigator)) ; Same conditions used to check if timer is valid for personal timer
  (if (and _bresult (Utilities|Enum|NotEqual(Enum) (Variables|UserOptions|GetSaveBehaviorKnob) "NewEnumerator0"))
    (Persistence|ClearData (Variables|Default|GetFortCreativePersistenceComponentBase) (PlayerState|GetUniqueNetId (Variables|/Script/Engine.Controller|GetPlayerState Instigator)) (Variables|Persistence|GetSavePath))))

(event Custom|ClearPersistenceDataForAllFromEvent
  (params (Instigator :type Controller%20Object%20Reference))
  ;[0.09255 1 0.3687] Channel Receivers for Save/Load
  (bind _bresult (Default|CanTriggerAllPersonalTimers Instigator))
  (if (and _bresult (Utilities|Enum|NotEqual(Enum) (Variables|UserOptions|GetSaveBehaviorKnob) "NewEnumerator0"))
    (Utilities|IsValid (Variables|/Script/FortniteGame.FortMinigameLogicComponent|GetCurrentMinigame (Variables|Components|GetFortMinigameProgress))
      (:Is%20Valid
        (for _array_element (Persistence|GetPersistencePlayers (Variables|Default|GetFortCreativePersistenceComponentBase) :!class /Script/FortniteGame.FortCreativePersistenceComponentBase)
          (Persistence|ClearData (Variables|Default|GetFortCreativePersistenceComponentBase) (PlayerState|GetUniqueNetId _array_element) (Variables|Persistence|GetSavePath)))))))

(event OnAddedtoMinigame(FortMinigameProgress)
  (params (Minigame :type Fort%20Minigame%20Object%20Reference))
  (Networking|SwitchHasAuthority
    (:Authority
      ;[1 1 1] Player removed needed so we can clear reference if the referenced player leaves the game
      (Minigame|BindEventtoPlayerRemovedDelegate Minigame (EventDispatchers|CreateEvent :!event OnPlayerRemovedFromMinigame))
      (Minigame|BindEventtoPlayerLastChancetoSave Minigame (EventDispatchers|CreateEvent :!event On%20Last%20Chance%20to%20Save)))))

(event OnRemovedfromMinigame(FortMinigameProgress)
  (params (Minigame :type Fort%20Minigame%20Object%20Reference))
  (Networking|SwitchHasAuthority
    (:Authority
      (Minigame|UnbindEventfromPlayerRemovedDelegate Minigame (EventDispatchers|CreateEvent :!event OnPlayerRemovedFromMinigame))
      (Minigame|UnbindEventfromPlayerLastChancetoSave Minigame (EventDispatchers|CreateEvent :!event On%20Last%20Chance%20to%20Save)))))

(event MinigameEndedDelegate(FortMinigameProgress)
  (Networking|SwitchHasAuthority
    (:Authority
      ;[1 1 1] Clear Last Instigator
      (Variables|Default|SetLastInstigator 0))))

(event OnVisibilityChanged(TextVisibileInGameComponent)
  (params (Visible :type Boolean))
  (Class|CreativeGenericTextWidgetComp|UpdateVisibilityandUpdateWidget (Variables|Default|GetTimerWidget) Visible))

(event Custom|BindClientExitVolumeGameplayMessage
  ;[1 1 1] Bind to Gameplay Message for exit volume
  (bind (_asynctaskproxy _payload) (GameplayEventRouter|StartListeningToEventBackwardCompatible "(TagName=\"EventMessage.Component.Playspace.PlayerExitedSpatialActor.LegacyMinigame\")" (GameplayEventRouter|GetGlobalEventRouter) "/Script/FortniteGame.PlayerExitSpatialActorContextLegacyMinigame" 0 "Down")
    (:then
      (Variables|Default|SetExitedEventHandle _asynctaskproxy))
    (:OnEventReceived
      (Delegates|OnClientExitVolume (.ExitingPlayerState (Utilities|Struct|BreakPlayerExitSpatialActorContextLegacyMinigame _payload)) (Utilities|Struct|BreakPlayerExitSpatialActorContextLegacyMinigame _payload)))))

(event Custom|TriggerUrgentFX
  (bind _returnvalue_x (Math|Vector|BreakVector (Rendering|Components|TextRender|GetTextWorldSize (Variables|Components|GetTimerDisplayTextFront))))
  (Utilities|FlowControl|Sequence
    (:then_0
      (if (not (Variables|Cosmetic|GetbUrgencyHasStarted))
        (Variables|Cosmetic|SetbUrgencyHasStarted true)
        (Visual|UpdateTimerColor (Variables|Default|GetUrgencyModeColor))
        (if (Math|Boolean|ANDBoolean (not (Audio|Components|Audio|IsPlaying (Variables|Default|GetUrgentBuildAudio))) (Math|Conversions|ToBoolean(Integer) (Variables|UserOptions|GetSounds)) (or (Math|Conversions|ToBoolean(Integer) (Variables|UserOptions|GetShowOnHUDKnob)) (Variables|UserOptions|GetbIsVisibleDuringGamesKnob)))
          (Audio|Components|Audio|Play (Variables|Default|GetUrgentBuildAudio) 0.0))))
    (:then_1
      (if (and (Utilities|IsValid (Variables|Default|GetUrgencySound) :!class /Script/Engine.KismetSystemLibrary) (Math|Conversions|ToBoolean(Integer) (Variables|UserOptions|GetSounds)))
        (if (Math|Conversions|ToBoolean(Integer) (Variables|UserOptions|GetShowOnHUDKnob))
          (Audio|PlaySound2D (Variables|Default|GetUrgencySound) 1.0 1.0 0.0 0 self false)
          (elif (Variables|UserOptions|GetbIsVisibleDuringGamesKnob)
            (Audio|PlaySoundatLocation (Variables|Default|GetUrgencySound) (Transformation|GetActorLocation) (Transformation|GetActorRotation) 1.0 1.0 0.0 0 0 self)))))
    (:then_2
      (Rendering|Material|SetScalarParameterValueonMaterials (Variables|Default|GetUrgencyGlow) (Variables|Default|GetUrgencyGlowScaleXParameter) (* _returnvalue_x (Variables|Default|GetUrgencyGlowScaleMultiplier)))
      (CallFunction|FlashVFXTimeline))))

(event Custom|FlashVFXTimeline
  (bind (_direction _lerp) (Timeline|Timeline_0 PlayFromStart)
    (:Update
      (Rendering|Material|SetScalarParameterValueonMaterials (Variables|Default|GetUrgencyGlow) (Variables|Default|GetUrgencyGlowOpacityParameter) _lerp))))

(fn AwardScore
  (params (TrackedPlayer :type Fort%20Player%20Controller%20Object%20Reference) (TimeRemaining :type Float%20%28double-precision%29))
  (bind (_outsourcetags _outcontexttags) (Class|FortMinigameStatsComponent|GetSourceandContextTags (LocalVariables|Getlocal%20MinigameStatsComp) (LocalVariables|Getlocal%20FortPlayerController)))
  (LocalVariables|Setlocal%20FortPlayerController TrackedPlayer)
  (bind _local_fortplayercontroller (LocalVariables|Getlocal%20FortPlayerController)
    (:then
      (bind _returnvalue (CreativeDevices|GetMinigameStatsComp _local_fortplayercontroller "Athena"))
      (LocalVariables|Setlocal%20MinigameStatsComp _returnvalue)
      (LocalVariables|Getlocal%20MinigameStatsComp
        (:then
          (bind _output_get (LocalVariables|SetScore (+ (Math|Integer|SelectInt (Variables|UserOptions|GetSuccessScoreKnob) (Variables|UserOptions|GetFailureScoreKnob) (CallFunction|TimerWasSuccessful TimeRemaining)) (Math|Integer|SelectInt (Math|Float|Round (* (Math|Float|Truncate (Variables|GetTimeRemaining)) (Variables|UserOptions|GetScorePerSecondRemainingKnob))) 0 (> (Variables|GetTimeRemaining) 0.0)))))
          (if (!= _output_get 0)
            (bind _output_get_2 (LocalVariables|SetTarget%20Tags (GameplayTags|GetOwnedGameplayTags :!class /Script/GameplayTags.GameplayTagAssetInterface)))
            (GameplayTags|AddGameplayTag _output_get_2 "(TagName=\"Minigame.Stat.Score.Goal\")" :!class /Script/GameplayTags.BlueprintGameplayTagLibrary)
            (Class|FortMinigameStatsComponent|SendStatEventwithTags (LocalVariables|Getlocal%20MinigameStatsComp) (LocalVariables|Getlocal%20FortPlayerController) "MinigameDynamicEvent" self (LocalVariables|GetTarget%20Tags) _outsourcetags _outcontexttags (LocalVariables|GetScore))
            (elif (and (Variables|UserOptions|Advanced|GetDisplayScoreUpdateIfScoreIs0) (Variables|UserOptions|Advanced|GetDisplayScoreUpdateOnHUD))
              _))
          (Class|CreativeHUDBroadcastComponent|DoScoreEvent (Variables|Default|GetCreative_HUDBroadcast_Component) (Variables|GetTrackedPlayer) (LocalVariables|GetScore) (Variables|UserOptions|Advanced|GetDisplayScoreUpdateOnHUD)))
        (:else)))
    (:else)))

(fn BindDelegates
  (bind _fortminigameprogress (Variables|Components|GetFortMinigameProgress))
  (Minigame|BindEventtoMinigameStateChangedDelegate _fortminigameprogress (EventDispatchers|CreateEvent :!event OnMinigameStateChanged))
  (Minigame|BindEventtoMinigameStartedDelegate _fortminigameprogress (EventDispatchers|CreateEvent :!event OnMinigameStarted))
  (Minigame|BindEventtoMinigameEndedDelegate _fortminigameprogress (EventDispatchers|CreateEvent :!event OnMinigameEnded)))

(fn CanClassInteract
  (params (Player :type Controller%20Object%20Reference))
  (return (StateChecks|IsValidClass (Variables|Default|GetCreative_UseRestriction_Component) Player)))

(fn CanPlayerAutoSave
  (params (FortPlayerState :type Fort%20Player%20State%20Object%20Reference))
  (return false))

(fn CanTriggerAllPersonalTimers
  (params (Instigator :type Controller%20Object%20Reference))
  (return (Math|Boolean|ANDBoolean (Utilities|Enum|Equal(Enum) (Variables|UserOptions|GetAppliesToKnob) "NewEnumerator0") (or (Knobs|Team|IsTeamAny (Variables|UserOptions|GetTeamSetting)) (Math|Byte|Equal(Byte) (Team|GetActorTeam Instigator) (Variables|Default|GetDeviceTeamID))) (CallFunction|CanClassInteract Instigator))))

(fn CheckFailedResetOrDisable
  (params (TrackedPlayer :type Controller%20Object%20Reference))
  (if (Variables|UserOptions|GetbDisableTimerOnFailedCheckKnob)
    (CallFunction|DisableTimer TrackedPlayer)
    (elif (Variables|UserOptions|GetbResetTimerOnFailedCheckKnob)
      (CallFunction|ResetTimer TrackedPlayer))))

(fn ClearDisplayTexts
  (bind (_displayscoreupdateonhud _hud_message _hud_messagescorecolor _hud_messagecolor _resethud_messagescore _displayscoreupdateifscoreis0 _bdisabletimeronfailedcheckknob _bresettimeronfailedcheckknob _activatingplayer _clientcurrenttime) (Class|GetClassDefaults (Utilities|GetClass self)))
  (Rendering|Components|TextRender|SetText (Variables|Components|GetTimerDisplayTextFront) (Utilities|Text|FormatText ""))
  (Utilities|Time|ClearandInvalidateTimerbyHandle (Variables|Cosmetic|GetClientDisplayUpdateTimerHandle))
  (Variables|FortCreativeTimerDevice|SetClientCurrentSecondaryText "")
  (Class|CreativeGenericTextWidgetComp|UpdateText (Variables|Default|GetSecondaryWidget) "")
  (Visual|UpdateTimerColor (Variables|UserOptions|GetTimerColorKnob))
  ;[0.8129 0.6457 1] Reset display time back to class default to allow inactive timers to have their worldspace text widget initially display on subsequent minigames
  (Variables|FortCreativeTimerDevice|SetClientCurrentTime _clientcurrenttime)
  (if (Math|Conversions|ToBoolean(Integer) (Variables|UserOptions|GetShowOnHUDKnob))
    (CallFunction|BroadcastUpdatetoHUD)))

(fn DisableTimer
  (params (TrackedPlayer :type Controller%20Object%20Reference))
  (switch Utilities|FlowControl|Switch|SwitchonETimerDeviceState (CallFunction|GetTimerState TrackedPlayer)
    (:Enabled
      _)
    (:Activated
      _)
    (:Paused
      _)
    (:Completed
      _))
  (CallFunction|TimerSetState TrackedPlayer "Disabled"))

(fn FilterBotsFromFortPlayerStates
  (params (InFortPlayerStates :type Array%20of%20Fort%20Player%20State%20Object%20References))
  ;[0 1 0.0938 0.5] Returns a copy of the input array without bots.
  (for _array_element InFortPlayerStates
    (Utilities|IsValid _array_element
      (:Is%20Valid
        (if (not (Variables|/Script/Engine.PlayerState|GetbIsABot _array_element))
          (Utilities|Array|Add (LocalVariables|GetNonBotPlayersLocal) _array_element)))))
  (return (LocalVariables|GetNonBotPlayersLocal)))

(fn GetIntendedTrackedPlayer
  (params (Instigator :type Controller%20Object%20Reference))
  (if (CallFunction|CanClassInteract Instigator)
    (switch Utilities|FlowControl|Switch|SwitchonE_Timer_AppliesTo (Variables|UserOptions|GetAppliesToKnob)
      (:NewEnumerator0
        ;[1 1 1] Get individually tracked player
        (return (and (Utilities|Array|ContainsItem (CallFunction|GetTrackedPlayers) (Variables|Actor|GetInstigator)) (Utilities|IsValid (Variables|Actor|GetInstigator) :!class /Script/Engine.KismetSystemLibrary)) (Variables|Actor|GetInstigator)))
      (:NewEnumerator1
        (if (Knobs|Team|IsTeamAny (Variables|UserOptions|GetTeamSetting))
          _
          (elif (Math|Byte|Equal(Byte) (Team|GetActorTeam (Variables|Actor|GetInstigator)) (Variables|Default|GetDeviceTeamID))
            _
            (else
              ;[1 1 1] Instigator is not being tracked
              (return false 0))))
        ;[1 1 1] Global timer with null instigator
        (return true 0)))
    (else
      (return false 0))))

(fn GetIntendedUpdateDisplayPlayers
  (params (TrackedPlayer :type Controller%20Object%20Reference) (OutPlayersToUpdate :type Array%20of%20Controller%20Object%20References))
  (switch Utilities|FlowControl|Switch|SwitchonE_Timer_AppliesTo (Variables|UserOptions|GetAppliesToKnob)
    (:NewEnumerator0
      (Utilities|Array|Add (Variables|GetOutPlayersToUpdate) (Variables|GetTrackedPlayer)))
    (:NewEnumerator1
      (bind _currentminigame (Variables|/Script/FortniteGame.FortMinigameLogicComponent|GetCurrentMinigame (Variables|Components|GetFortMinigameProgress))
        (:then
          (if (Knobs|Team|IsTeamAny (Variables|UserOptions|GetTeamSetting))
            (for _array_element (Minigame|GetParticipatingPlayers _currentminigame)
              (bind _ascontroller (Utilities|Casting|CastToController (Actor|GetOwner _array_element)))
              (Utilities|Array|Add (Variables|GetOutPlayersToUpdate) _ascontroller))
            (else
              (Utilities|IsValid (Minigame|GetMinigameTeamComponent _currentminigame)
                (:Is%20Valid
                  (for _array_element_2 (Class|FortMinigameTeamComponent|GetTeamMembers (Minigame|GetMinigameTeamComponent _currentminigame) (Variables|Default|GetDeviceTeamID))
                    (bind _ascontroller_2 (Utilities|Casting|CastToController (Actor|GetOwner _array_element_2)))
                    (Utilities|Array|Add (Variables|GetOutPlayersToUpdate) _ascontroller_2)))))))
        (:else)))))

(fn GetObjectToAutoSaveForPlayer
  (params (FortPlayerState :type Fort%20Player%20State%20Object%20Reference))
  (return 0))

(fn GetPersistenceRequiredState
  (return (and (Utilities|Enum|Equal(Enum) (Variables|UserOptions|GetAppliesToKnob) "NewEnumerator0") (Utilities|Enum|NotEqual(Enum) (Variables|UserOptions|GetSaveBehaviorKnob) "NewEnumerator0"))))

(fn InitializeTrackedTimers
  (bind _fortminigameprogress (Variables|Components|GetFortMinigameProgress))
  (bind _currentminigame (Variables|/Script/FortniteGame.FortMinigameLogicComponent|GetCurrentMinigame _fortminigameprogress))
  (bind _returnvalue (Minigame|GetMinigameTeamComponent _currentminigame))
  (bind _fortminigameprogress_2 (Variables|Components|GetFortMinigameProgress))
  (bind _currentminigame_2 (Variables|/Script/FortniteGame.FortMinigameLogicComponent|GetCurrentMinigame _fortminigameprogress_2))
  (Networking|SwitchHasAuthority
    (:Authority
      (Utilities|FlowControl|Sequence
        (:then_0
          (CallFunction|ClearTrackedTimers)
          (Utilities|Map|Clear (Variables|Default|GetStateBeforeDisabledCache)))
        (:then_1
          (switch Utilities|FlowControl|Switch|SwitchonE_Timer_AppliesTo (Variables|UserOptions|GetAppliesToKnob)
            (:NewEnumerator0
              (if (Knobs|Team|IsTeamAny (Variables|UserOptions|GetTeamSetting))
                ;[1 1 1] Get every player and make a timer for each
                (Utilities|IsValid _currentminigame_2
                  (:Is%20Valid
                    (for _array_element (Minigame|GetParticipatingPlayers _currentminigame_2)
                      (bind _ascontroller (Utilities|Casting|CastToController (Actor|GetOwner _array_element)))
                      (Default|StartTrackingPlayer _ascontroller))))
                (else
                  ;[1 1 1] Get every player on the team and make a timer for each
                  (Utilities|IsValid _currentminigame
                    (:Is%20Valid
                      (Utilities|IsValid _returnvalue
                        (:Is%20Valid
                          (for _array_element_2 (Class|FortMinigameTeamComponent|GetTeamMembers _returnvalue (Variables|Default|GetDeviceTeamID))
                            (bind _ascontroller_2 (Utilities|Casting|CastToController (Actor|GetOwner _array_element_2)))
                            (Default|StartTrackingPlayer _ascontroller_2)))))))))
            (:NewEnumerator1
              ;[1 1 1] Use a single timer with no instigator
              (Default|StartTrackingPlayer 0))))))))

(fn IsTargetableByLinearMover
  (return true))

(fn LoadTimer
  (params (PlayerController :type Controller%20Object%20Reference) (bUseTimeSinceSave :type Boolean))
  (bind _timesincesave (LocalVariables|GetTimeSinceSave))
  (bind _returnvalue (- (Variables|/Game/Creative/Devices/TimerDevice/Persistence/PersistenceData_Timer.PersistenceData_Timer_C|GetTimeRemaining (Variables|Persistence|GetPersistenceData_Timer)) _timesincesave))
  (bind _persistencedata_timer (Variables|Persistence|GetPersistenceData_Timer))
  (bind _persistencedata_timer_2 (Variables|Persistence|GetPersistenceData_Timer))
  (bind _timeremaining (Variables|/Game/Creative/Devices/TimerDevice/Persistence/PersistenceData_Timer.PersistenceData_Timer_C|GetTimeRemaining _persistencedata_timer_2))
  (Networking|SwitchHasAuthority
    (:Authority
      (if (and (Utilities|Enum|Equal(Enum) (Variables|UserOptions|GetAppliesToKnob) "NewEnumerator0") (Utilities|Enum|NotEqual(Enum) (Variables|UserOptions|GetSaveBehaviorKnob) "NewEnumerator0"))
        (if (CallFunction|CanClassInteract (Variables|GetPlayerController))
          (Variables|/Game/Creative/Devices/TimerDevice/Persistence/PersistenceData_Timer.PersistenceData_Timer_C|SetbIsDataLoaded false _persistencedata_timer)
          (Persistence|LoadData (Variables|Default|GetFortCreativePersistenceComponentBase) (PlayerState|GetUniqueNetId (Variables|/Script/Engine.Controller|GetPlayerState (Variables|GetPlayerController))) (Variables|Persistence|GetSavePath) _persistencedata_timer)
          (if (Variables|/Game/Creative/Devices/TimerDevice/Persistence/PersistenceData_Timer.PersistenceData_Timer_C|GetbIsDataLoaded _persistencedata_timer)
            (Utilities|FlowControl|Sequence
              (:then_0
                ;[1 1 1] Upon new round the timer map is cleared, so we may need to recreate the tracked timer
                (Default|StartTrackingPlayer (Variables|GetPlayerController)))
              (:then_1
                ;[1 1 1] Reset timer time to saved value, and if necessary subtract out time since last save and get new time remaining
                (LocalVariables|SetTimeSinceSave (Math|Float|SelectFloat (- (Fortnite|Sync|Time|GetServerWorldTimeSeconds) (Variables|/Game/Creative/Devices/TimerDevice/Persistence/PersistenceData_Timer.PersistenceData_Timer_C|GetServerTime (Variables|Persistence|GetPersistenceData_Timer))) 0.0 (Variables|GetbUseTimeSinceSave)))
                (LocalVariables|SetbTimerCompletedSinceSave (<= _returnvalue 0.0))
                (CallFunction|TimerStart (Variables|GetPlayerController) _returnvalue (- (+ (LocalVariables|GetTimeSinceSave) (- (Variables|UserOptions|GetDurationKnob) _timeremaining)) (- (Variables|UserOptions|GetDurationKnob) (Variables|UserOptions|GetUrgencyModeTimeKnob))))
                (CallFunction|TimerPause (Variables|GetPlayerController)))
              (:then_2
                ;[1 1 1] Save over the disabled cache if it was updated after saving a disabled timer, or clear it if a disabled timer is loaded to a non disabled state
                (if (and (Utilities|Enum|Equal(Enum) (Variables|/Game/Creative/Devices/TimerDevice/Persistence/PersistenceData_Timer.PersistenceData_Timer_C|GetState (Variables|Persistence|GetPersistenceData_Timer)) "Disabled") (or (Utilities|Enum|Equal(Enum) (Variables|/Game/Creative/Devices/TimerDevice/Persistence/PersistenceData_Timer.PersistenceData_Timer_C|GetStateBeforeDisabled (Variables|Persistence|GetPersistenceData_Timer)) "Activated") (Utilities|Enum|Equal(Enum) (Variables|/Game/Creative/Devices/TimerDevice/Persistence/PersistenceData_Timer.PersistenceData_Timer_C|GetStateBeforeDisabled (Variables|Persistence|GetPersistenceData_Timer)) "Paused")))
                  (Utilities|Map|Add (Variables|Default|GetStateBeforeDisabledCache) (Variables|GetPlayerController) (Variables|/Game/Creative/Devices/TimerDevice/Persistence/PersistenceData_Timer.PersistenceData_Timer_C|GetStateBeforeDisabled (Variables|Persistence|GetPersistenceData_Timer)))
                  (else
                    (Utilities|Map|Remove (Variables|Default|GetStateBeforeDisabledCache) (Variables|GetPlayerController)))))
              (:then_3
                (if (LocalVariables|GetbTimerCompletedSinceSave)
                  (CallFunction|TimerSetState (Variables|GetPlayerController) "Completed")
                  (else
                    (CallFunction|TimerSetState (Variables|GetPlayerController) (Variables|/Game/Creative/Devices/TimerDevice/Persistence/PersistenceData_Timer.PersistenceData_Timer_C|GetState (Variables|Persistence|GetPersistenceData_Timer)) true)))
                (return true)))
            (else
              (return false))))))))

(fn MessageScore
  (return (Variables|UserOptions|Advanced|GetHUD%20Message) (Variables|UserOptions|Advanced|GetHUD%20MessageScoreColor) (Variables|UserOptions|Advanced|GetHUD%20MessageColor) (Variables|UserOptions|Advanced|GetResetHUD%20MessageScore)))

(fn OnLastChancetoSave
  (params (PlayerUniqueNetId :type Unique%20Net%20Id%20Structure) (bIsLocalPlayer :type Boolean))
  (bind _playeruniquenetid (Variables|GetPlayerUniqueNetId))
  (bind _returnvalue (Game|GetPlayerStatefromUniqueNetId _playeruniquenetid))
  (bind _returnvalue_2 (PlayerState|GetPlayerController _returnvalue))
  (Networking|SwitchHasAuthority
    (:Authority
      (if (Math|Boolean|ANDBoolean (Utilities|Enum|NotEqual(Enum) (Variables|UserOptions|GetAppliesToKnob) "NewEnumerator1") (Utilities|Enum|NotEqual(Enum) (Variables|UserOptions|GetSaveBehaviorKnob) "NewEnumerator0") (Variables|UserOptions|GetbAutoSaveKnob))
        (Utilities|IsValid _returnvalue
          (:Is%20Valid
            (Utilities|IsValid _returnvalue_2
              (:Is%20Valid
                (Persistence|SaveTimer _returnvalue_2)
                (CallFunction|TimerSetState _returnvalue_2 "Paused")
                (CallFunction|ServerRequestClearDisplay (Utilities|Array|MakeArray _returnvalue_2))))))))))

(fn OnClassRemoved
  (params (PlayerState :type Fort%20Player%20State%20Object%20Reference) (OldClassId :type Byte))
  (bind _playerstate (Variables|GetPlayerState))
  (Networking|SwitchHasAuthority
    (:Authority
      (if (ClassRestrictions|TestMatchingClass (Variables|Default|GetCreative_UseRestriction_Component) (Math|Conversions|ToInteger(Byte) (Variables|GetOldClassId)))
        (Utilities|IsValid _playerstate
          (:Is%20Valid
            (bind _ascontroller (Utilities|Casting|CastToController (Actor|GetOwner _playerstate)))
            (switch Utilities|FlowControl|Switch|SwitchonE_Timer_AppliesTo (Variables|UserOptions|GetAppliesToKnob)
              (:NewEnumerator0
                ;[1 1 1] Player leaving this device's class
                (CallFunction|CheckFailedResetorDisable _ascontroller))
              (:NewEnumerator1
                ;[1 1 1] Universal timer when ActivatingPlayer leaves this device's class
                (bind _activatingplayer (Variables|Default|GetActivatingPlayer)
                  (:then
                    (if (== _activatingplayer _ascontroller)
                      (CallFunction|CheckFailedResetorDisable 0) ; We don't track player here because a unversal timer tracks its widgets with a null
                    ))
                  (:else))))))))))

(fn OnClientDisplayUpdateTimerCompleted
  (Visual|UpdateinGameDisplay "Activated" (- (Variables|Cosmetic|GetClientDisplayUpdateEndTime) (Utilities|Time|GetTimeSeconds))))

(fn OnClientExitVolume
  (params (Client :type Player%20State%20Object%20Reference) (Volume :type Actor%20Object%20Reference))
  (Networking|SwitchHasAuthority
    (:Authority
      ;[1 1 1] Check that we're exiting our specific island, and that this is not a global timer
      (if (Math|Boolean|ANDBoolean (Utilities|IsValid (Minigame|GetMinigame Volume :!class /Script/FortniteGame.MinigameBlueprintLibrary) :!class /Script/Engine.KismetSystemLibrary) (Utilities|IsValid (Variables|/Script/FortniteGame.FortMinigameLogicComponent|GetCurrentMinigame (Variables|Components|GetFortMinigameProgress)) :!class /Script/Engine.KismetSystemLibrary) (Utilities|Equal(Object) (Minigame|GetMinigame Volume :!class /Script/FortniteGame.MinigameBlueprintLibrary) (Variables|/Script/FortniteGame.FortMinigameLogicComponent|GetCurrentMinigame (Variables|Components|GetFortMinigameProgress))) (Utilities|Enum|NotEqual(Enum) (Variables|UserOptions|GetAppliesToKnob) "NewEnumerator1"))
        (bind _ascontroller (Utilities|Casting|CastToController (Actor|GetOwner (Variables|GetClient))))
        (if (not (and (Variables|UserOptions|GetbAutoSaveKnob) (Utilities|Enum|NotEqual(Enum) (Variables|UserOptions|GetSaveBehaviorKnob) "NewEnumerator0")))
          (Default|StopTrackingPlayer _ascontroller))))))

(fn OnMinigameEnded
  (bind _fortminigameprogress (Variables|Components|GetFortMinigameProgress))
  (bind _currentminigame (Variables|/Script/FortniteGame.FortMinigameLogicComponent|GetCurrentMinigame _fortminigameprogress))
  (Networking|SwitchHasAuthority
    (:Authority
      (Variables|Default|SetLastInstigator 0)
      (Minigame|UnbindEventfromPlayerChangedTeamDelegate _currentminigame (EventDispatchers|CreateEvent :!event OnTeamChanged))
      (Minigame|UnbindEventfromOnPlayerClassRemoved _currentminigame (EventDispatchers|CreateEvent :!event OnClassRemoved))
      (Minigame|UnbindEventfromOnPlayerJoininProgress _currentminigame (EventDispatchers|CreateEvent :!event OnPlayerJoinInProgress))
      (CallFunction|ClearTrackedTimers))
    (:Remote
      (Visual|ClearDisplayTexts))))

(fn OnMinigameStarted
  (bind _fortminigameprogress (Variables|Components|GetFortMinigameProgress))
  (bind _currentminigame (Variables|/Script/FortniteGame.FortMinigameLogicComponent|GetCurrentMinigame _fortminigameprogress))
  (Variables|Default|SetLastInstigator 0)
  (Minigame|BindEventtoPlayerChangedTeamDelegate _currentminigame (EventDispatchers|CreateEvent :!event OnTeamChanged))
  (Minigame|BindEventtoOnPlayerClassRemoved _currentminigame (EventDispatchers|CreateEvent :!event OnClassRemoved))
  (Minigame|BindEventtoOnPlayerJoininProgress _currentminigame (EventDispatchers|CreateEvent :!event OnPlayerJoinInProgress))
  (Default|InitializeTrackedTimers))

(fn OnMinigameStateChanged
  (params (Minigame :type Fort%20Minigame%20Object%20Reference) (NewMinigameState :type EFortMinigameState%20Enum :default "PreGame"))
  (switch Utilities|FlowControl|Switch|SwitchonEFortMinigameState NewMinigameState
    (:PreGame
      _)
    (:Setup
      _)
    (:PostRoundEnd
      _))
  (Networking|CanExecuteCosmeticEvents
    (:True
      (Visual|ClearDisplayTexts))))

(fn OnPlayerJoinInProgress
  (params (Pawn :type Pawn%20Object%20Reference))
  (bind _pawn (Variables|GetPawn))
  (bind _returnvalue (Pawn|GetController _pawn))
  (Networking|SwitchHasAuthority
    (:Authority
      (if (Utilities|Enum|NotEqual(Enum) (Variables|UserOptions|GetAppliesToKnob) "NewEnumerator1")
        (if (Knobs|Team|IsTeamAny (Variables|UserOptions|GetTeamSetting))
          _
          (elif (Math|Byte|Equal(Byte) (Variables|Default|GetDeviceTeamID) (Team|GetActorTeam (Variables|GetPawn)))
            _))
        (if (and (Variables|UserOptions|GetbAutoSaveKnob) (Variables|UserOptions|GetbAutoLoadKnob))
          (bind _bloadeddata (Persistence|LoadTimer _returnvalue (Utilities|Enum|Equal(Enum) (Variables|UserOptions|GetSaveBehaviorKnob) "NewEnumerator7")))
          (if _bloadeddata
            _
            (else
              _))
          (else
            _))
        (Default|StopTrackingPlayer _returnvalue) ; If Auto-Save enabled but Auto-Load is not, need to discard existing paused timer before starting to track again
        (Default|StartTrackingPlayer _returnvalue)))))

(fn OnPlayerRemovedFromMinigame
  (params (Unique%20Net%20Id :type Unique%20Net%20Id%20Structure) (bIsLocalPlayer :type Boolean))
  (bind _lastinstigator (Variables|Default|GetLastInstigator))
  (Networking|SwitchHasAuthority
    (:Authority
      (Utilities|IsValid _lastinstigator
        (:Is%20Valid
          (bind _asfort_player_state (Utilities|Casting|CastToFortPlayerState (Variables|/Script/Engine.Controller|GetPlayerState _lastinstigator)))
          (if (== (PlayerState|GetPlayerUniqueNetID _asfort_player_state) Unique%20Net%20Id)
            (Variables|Default|SetbInstigatorWasRemoved true)
            (if (Utilities|Enum|Equal(Enum) (Variables|Default|GetInstigatingPlayerOption) "NewEnumerator0")
              ;[1 1 1] Clear Last Instigator
              (Variables|Default|SetLastInstigator 0)
              (else
                (bind _playercontroller (Utilities|GetRandomTeamMember)
                  (:HasValidMember
                    (Variables|Default|SetLastInstigator _playercontroller))
                  (:NoValidMember
                    (Variables|Default|SetLastInstigator 0)))))))))))

(fn OnServerDisplayUpdateTimerCompleted
  (params (TrackedPlayer :type Controller%20Object%20Reference))
  (Default|GetIntendedUpdateDisplayPlayers TrackedPlayer (LocalVariables|GetPlayersToUpdate))
  (CallFunction|ServerRequestDisplayUpdate TrackedPlayer (LocalVariables|GetPlayersToUpdate)))

(fn OnTeamChanged
  (params (PlayerState :type Fort%20Player%20State%20Object%20Reference) (TeamId :type Byte) (OldTeamId :type Byte) (bTeamChangedWithoutRespawn :type Boolean))
  (bind _playerstate (Variables|GetPlayerState))
  (bind _playerstate_2 (Variables|GetPlayerState))
  (bind _playerstate_3 (Variables|GetPlayerState))
  (bind _playerstate_4 (Variables|GetPlayerState))
  (bind _playerstate_5 (Variables|GetPlayerState))
  (Networking|SwitchHasAuthority
    (:Authority
      ;[1 1 1] Check that this is a team timer
      (if (not (Knobs|Team|IsTeamAny (Variables|UserOptions|GetTeamSetting)))
        (Utilities|FlowControl|Sequence
          (:then_0
            ;[1 1 1] Handle personal timers
            (switch Utilities|FlowControl|Switch|SwitchonE_Timer_AppliesTo (Variables|UserOptions|GetAppliesToKnob)
              (:NewEnumerator0
                ;;[1 1 1] Player joining this device's team
                (if (Math|Byte|Equal(Byte) (Variables|GetTeamId) (Variables|Default|GetDeviceTeamID))
                  (Utilities|IsValid _playerstate_2
                    (:Is%20Valid
                      (bind _ascontroller (Utilities|Casting|CastToController (Actor|GetOwner _playerstate_2)))
                      (Default|StartTrackingPlayer _ascontroller)))
                  (elif (Math|Byte|Equal(Byte) (Variables|GetOldTeamId) (Variables|Default|GetDeviceTeamID))
                    ;;[1 1 1] Player leaving this device's team
                    (Utilities|IsValid _playerstate_3
                      (:Is%20Valid
                        (bind _ascontroller_2 (Utilities|Casting|CastToController (Actor|GetOwner _playerstate_3)))
                        (CallFunction|CheckFailedResetorDisable _ascontroller_2)
                        (Default|StopTrackingPlayer _ascontroller_2))))))
              (:NewEnumerator1
                ;[1 1 1] Handle universal timers
                ;;[1 1 1] Player joining this device's team
                (if (Math|Byte|Equal(Byte) (Variables|GetTeamId) (Variables|Default|GetDeviceTeamID))
                  (Utilities|IsValid _playerstate_5
                    (:Is%20Valid
                      (bind _ascontroller_3 (Utilities|Casting|CastToController (Actor|GetOwner _playerstate_5)))
                      (CallFunction|ServerRequestDisplayUpdate 0 (Utilities|Array|MakeArray _ascontroller_3))))
                  (elif (Math|Byte|Equal(Byte) (Variables|GetOldTeamId) (Variables|Default|GetDeviceTeamID))
                    ;;[1 1 1] Team-wide timer when ActivatingPlayer leaves this device's team
                    (Utilities|IsValid _playerstate_4
                      (:Is%20Valid
                        (bind _ascontroller_4 (Utilities|Casting|CastToController (Actor|GetOwner _playerstate_4)))
                        (bind _activatingplayer (Variables|Default|GetActivatingPlayer)
                          (:then
                            (if (== _activatingplayer _ascontroller_4)
                              (CallFunction|CheckFailedResetorDisable 0) ; We don't track player here because a unversal timer tracks its widgets with a null
                            ))
                          (:else)) ; Test that we have a valid Activating Player when we cache the last instigator on a StartTimer event
                      )))))))
          (:then_1
            ;[1 1 1] Client joining some other team, need to erase text
            (if (Math|Byte|NotEqual(Byte) (Variables|GetTeamId) (Variables|Default|GetDeviceTeamID))
              (Utilities|IsValid _playerstate
                (:Is%20Valid
                  (bind _ascontroller_5 (Utilities|Casting|CastToController (Actor|GetOwner _playerstate)))
                  (CallFunction|ServerRequestClearDisplay (Utilities|Array|MakeArray _ascontroller_5)))))))))))

(fn OnTimerCompleted
  (params (TrackedPlayer :type Controller%20Object%20Reference))
  (CallFunction|TimerSetState TrackedPlayer "Completed"))

(fn OnTimerStateChanged
  (params (TrackedPlayer :type Controller%20Object%20Reference) (TimerHandle :type Timer%20Handle%20Structure%20%28by%20ref%29) (CurrentState :type ETimerDeviceState%20Enum :default "Enabled") (PreviousState :type ETimerDeviceState%20Enum :default "Enabled") (bLoadingFromSave :type Boolean))
  (bind _trackedplayer (Variables|GetTrackedPlayer))
  (bind _statebeforedisabledcache (Variables|Default|GetStateBeforeDisabledCache))
  (bind (_value _returnvalue) (Utilities|Map|Find _statebeforedisabledcache _trackedplayer))
  (bind _timerhandle (Variables|GetTimerHandle))
  (bind _returnvalue_2 (Utilities|Time|GetTimerRemainingTimebyHandle _timerhandle))
  (bind _fortminigameprogress (Variables|Components|GetFortMinigameProgress))
  (bind _currentminigame (Variables|/Script/FortniteGame.FortMinigameLogicComponent|GetCurrentMinigame _fortminigameprogress))
  (bind _returnvalue_3 (Minigame|GetMinigameTeamComponent _currentminigame))
  (bind _timerhandle_2 (Variables|GetTimerHandle))
  (bind _fortminigameprogress_2 (Variables|Components|GetFortMinigameProgress))
  (bind _currentminigame_2 (Variables|/Script/FortniteGame.FortMinigameLogicComponent|GetCurrentMinigame _fortminigameprogress_2))
  (bind _oncompleteinstigator (LocalVariables|GetOnCompleteInstigator))
  (Utilities|FlowControl|Sequence
    (:then_0
      (switch Utilities|FlowControl|Switch|SwitchonETimerDeviceState (Variables|GetCurrentState)
        (:Enabled
          ;[0 0 1] Enabled
          (if (Math|Boolean|ANDBoolean (not (Variables|GetbLoadingFromSave)) (Utilities|Enum|Equal(Enum) (Variables|GetPreviousState) "Disabled") _returnvalue)
            (CallFunction|TimerSetState (Variables|GetTrackedPlayer) _value)
            (Utilities|Map|Remove (Variables|Default|GetStateBeforeDisabledCache) (Variables|GetTrackedPlayer))
            (else
              (CallFunction|TimerStart (Variables|GetTrackedPlayer) (Variables|UserOptions|GetDurationKnob) (- (Variables|UserOptions|GetDurationKnob) (Variables|UserOptions|GetUrgencyModeTimeKnob)))
              (CallFunction|TimerPause (Variables|GetTrackedPlayer)))))
        (:Activated
          ;[0 0.995 0] Activated
          (if (Utilities|Time|IsTimerPausedbyHandle _timerhandle_2)
            (CallFunction|TimerResume (Variables|GetTrackedPlayer))
            (elif (not (Utilities|Time|IsTimerActivebyHandle _timerhandle_2))
              (CallFunction|TimerStart (Variables|GetTrackedPlayer) (Variables|UserOptions|GetDurationKnob) (- (Variables|UserOptions|GetDurationKnob) (Variables|UserOptions|GetUrgencyModeTimeKnob))))))
        (:Paused
          ;[1 1 0] Paused
          (CallFunction|TimerPause (Variables|GetTrackedPlayer)))
        (:Completed
          ;[0 1 1] Completed
          (Utilities|FlowControl|Sequence
            (:then_0
              ;;[1 1 1] Handle Award Score Behavior
              (if (not (Variables|GetbLoadingFromSave))
                (switch Utilities|FlowControl|Switch|SwitchonE_Timer_AppliesTo (Variables|UserOptions|GetAppliesToKnob)
                  (:NewEnumerator0
                    (bind _asfort_player_controller (Utilities|Casting|CastToFortPlayerController (Variables|GetTrackedPlayer)))
                    (Default|AwardScore _asfort_player_controller (Utilities|Time|GetTimerRemainingTimebyHandle (Variables|GetTimerHandle))))
                  (:NewEnumerator1
                    (if (Knobs|Team|IsTeamAny (Variables|UserOptions|GetTeamSetting))
                      (Utilities|IsValid _currentminigame_2
                        (:Is%20Valid
                          (for _array_element (Minigame|GetParticipatingPlayers _currentminigame_2)
                            (bind _asfort_player_controller_2 (Utilities|Casting|CastToFortPlayerController (Actor|GetOwner _array_element)))
                            (Default|AwardScore _asfort_player_controller_2 _returnvalue_2))))
                      (else
                        (Utilities|IsValid _currentminigame
                          (:Is%20Valid
                            (Utilities|IsValid _returnvalue_3
                              (:Is%20Valid
                                (for _array_element_2 (Class|FortMinigameTeamComponent|GetTeamMembers _returnvalue_3 (Variables|Default|GetDeviceTeamID))
                                  (bind _asfort_player_controller_3 (Utilities|Casting|CastToFortPlayerController (Actor|GetOwner _array_element_2)))
                                  (Default|AwardScore _asfort_player_controller_3 _returnvalue_2))))))))))))
            (:then_1
              ;;[1 1 1] Handle Channel Trigger Behavior
              (if (Math|Boolean|ANDBoolean (not (Variables|GetbLoadingFromSave)) (Knobs|Team|IsTeamAny (Variables|UserOptions|GetTeamSetting)) (Utilities|Enum|Equal(Enum) (Variables|UserOptions|GetAppliesToKnob) "NewEnumerator1"))
                (bind _lastinstigator (Variables|Default|GetLastInstigator)
                  (:then
                    (LocalVariables|SetOnCompleteInstigator _lastinstigator))
                  (:else
                    (if (and (not (Variables|Default|GetbInstigatorWasRemoved)) (Utilities|Enum|Equal(Enum) (Variables|Default|GetInstigatingPlayerOption) "NewEnumerator7"))
                      (bind _playercontroller (Utilities|GetRandomTeamMember)
                        (:HasValidMember
                          (LocalVariables|SetOnCompleteInstigator _playercontroller))
                        (:NoValidMember
                          (LocalVariables|SetOnCompleteInstigator (Variables|GetTrackedPlayer))))
                      (else
                        (LocalVariables|SetOnCompleteInstigator 0)))))
                (Variables|Default|SetbInstigatorWasRemoved false)
                (if (CallFunction|TimerWasSuccessful (Utilities|Time|GetTimerRemainingTimebyHandle (Variables|GetTimerHandle)))
                  (GameplayEvent|BroadcastEvent (Variables|UserOptions|Events|GetOnSuccess) 0 (Utilities|Struct|MakeGameplayEventTimerSuccess _oncompleteinstigator))
                  (else
                    (GameplayEvent|BroadcastEvent (Variables|UserOptions|Events|GetOnFailure) 0 (Utilities|Struct|MakeGameplayEventTimerFailure _oncompleteinstigator))))
                (else
                  _)))
            (:then_2
              ;;[1 1 1] Handle On Complete Behavior
              (CallFunction|TimerPause (Variables|GetTrackedPlayer))
              (switch Utilities|FlowControl|Switch|SwitchonE_Timer_OnCompleteBehavior (Variables|UserOptions|GetCompletionBehaviorKnob)
                (:NewEnumerator0
                  (LocalVariables|SetOnCompleteNewState "Disabled"))
                (:NewEnumerator2
                  (LocalVariables|SetOnCompleteNewState "Enabled"))
                (:NewEnumerator3
                  (LocalVariables|SetOnCompleteNewState "Activated")))
              (LocalVariables|SetbOnCompleteChangeState true))))
        (:Disabled
          ;[1 0 0] Disabled
          (CallFunction|TimerPause (Variables|GetTrackedPlayer))
          (if (and (not (Variables|GetbLoadingFromSave)) (or (Utilities|Enum|Equal(Enum) (Variables|GetPreviousState) "Activated") (Utilities|Enum|Equal(Enum) (Variables|GetPreviousState) "Paused")))
            (Utilities|Map|Add (Variables|Default|GetStateBeforeDisabledCache) (Variables|GetTrackedPlayer) (Variables|GetPreviousState))))))
    (:then_1
      (Default|GetIntendedUpdateDisplayPlayers (Variables|GetTrackedPlayer) (LocalVariables|GetPlayersToUpdate))
      (CallFunction|ServerRequestDisplayUpdate (Variables|GetTrackedPlayer) (LocalVariables|GetPlayersToUpdate)))
    (:then_2
      (if (LocalVariables|GetbOnCompleteChangeState)
        (CallFunction|CompletionStateChangeDelay (Variables|GetTrackedPlayer) (LocalVariables|GetOnCompleteNewState))))))

(fn OnUrgencyTimeReached
  (params (TrackedPlayer :type Controller%20Object%20Reference))
  (Default|GetIntendedUpdateDisplayPlayers TrackedPlayer (LocalVariables|GetPlayersToUpdate))
  (CallFunction|ServerRequestDisplayUpdate (Variables|GetTrackedPlayer) (LocalVariables|GetPlayersToUpdate))
  (GameplayEvent|BroadcastEvent (Variables|UserOptions|Events|GetOnStartUrgencyMode) 0 (Utilities|Struct|MakeGameplayEventTimerUrgency (Variables|GetTrackedPlayer))))

(fn ResetTimer
  (params (TrackedPlayer :type Controller%20Object%20Reference))
  (switch Utilities|FlowControl|Switch|SwitchonETimerDeviceState (CallFunction|GetTimerState TrackedPlayer)
    (:Activated
      _)
    (:Paused
      _)
    (:Completed
      _))
  (CallFunction|TimerClearHandles TrackedPlayer)
  (CallFunction|TimerSetState TrackedPlayer))

(fn SaveTimer
  (params (PlayerController :type Controller%20Object%20Reference))
  (bind (_outstate _returnvalue) (CallFunction|GetTimerState (Variables|GetPlayerController)))
  (bind (_value _returnvalue_2) (Utilities|Map|Find (Variables|Default|GetStateBeforeDisabledCache) (Variables|GetPlayerController)))
  (bind _persistencedata_timer (Variables|Persistence|GetPersistenceData_Timer))
  (bind _playercontroller (Variables|GetPlayerController))
  (Networking|SwitchHasAuthority
    (:Authority
      (if (Math|Boolean|ANDBoolean (Utilities|Enum|Equal(Enum) (Variables|UserOptions|GetAppliesToKnob) "NewEnumerator0") (Utilities|Enum|NotEqual(Enum) (Variables|UserOptions|GetSaveBehaviorKnob) "NewEnumerator0") (or (Knobs|Team|IsTeamAny (Variables|UserOptions|GetTeamSetting)) (Math|Byte|Equal(Byte) (Variables|Default|GetDeviceTeamID) (Team|GetActorTeam _playercontroller))) (CallFunction|CanClassInteract _playercontroller) _returnvalue)
        (bind _output_get (Variables|/Game/Creative/Devices/TimerDevice/Persistence/PersistenceData_Timer.PersistenceData_Timer_C|SetState _outstate _persistencedata_timer))
        (Variables|/Game/Creative/Devices/TimerDevice/Persistence/PersistenceData_Timer.PersistenceData_Timer_C|SetStateBeforeDisabled (select (and _returnvalue_2 (Utilities|Enum|Equal(Enum) _output_get "Disabled")) _value "Enabled") _persistencedata_timer)
        (Variables|/Game/Creative/Devices/TimerDevice/Persistence/PersistenceData_Timer.PersistenceData_Timer_C|SetTimeRemaining (CallFunction|GetTimerRemaining (Variables|GetPlayerController)) _persistencedata_timer)
        (Variables|/Game/Creative/Devices/TimerDevice/Persistence/PersistenceData_Timer.PersistenceData_Timer_C|SetServerTime (Fortnite|Sync|Time|GetServerWorldTimeSeconds) _persistencedata_timer)
        (Variables|/Game/Creative/Devices/TimerDevice/Persistence/PersistenceData_Timer.PersistenceData_Timer_C|SetbIsDataLoaded true _persistencedata_timer)
        (Persistence|SaveData (Variables|Default|GetFortCreativePersistenceComponentBase) _persistencedata_timer (PlayerState|GetUniqueNetId (Variables|/Script/Engine.Controller|GetPlayerState (Variables|GetPlayerController))) (Variables|Persistence|GetSavePath) false)))))

(fn SetupClientDisplayUpdateTimer
  (params (CurrentState :type ETimerDeviceState%20Enum :default "Enabled") (TimeRemaining :type Float%20%28double-precision%29))
  (if (Utilities|Enum|Equal(Enum) CurrentState "Activated")
    (Variables|Cosmetic|SetClientDisplayUpdateEndTime (+ (Utilities|Time|GetTimeSeconds) (Variables|GetTimeRemaining)))
    (bind _returnvalue (Utilities|Time|SetTimerbyEvent (EventDispatchers|CreateEvent :!event OnClientDisplayUpdateTimerCompleted) (Variables|Cosmetic|GetClientDisplayUpdateRate) true false 0.0 0.3))
    (Variables|Cosmetic|SetClientDisplayUpdateTimerHandle _returnvalue)
    (else
      (Utilities|Time|ClearandInvalidateTimerbyHandle (Variables|Cosmetic|GetClientDisplayUpdateTimerHandle)))))

(fn ShouldInvisiblePropsBeAllowed
  (return false))

(fn StartTrackingPlayer
  (params (PlayerToTrack :type Controller%20Object%20Reference))
  (if (not (Utilities|Array|ContainsItem (CallFunction|GetTrackedPlayers) (Variables|GetPlayerToTrack)))
    (CallFunction|CreateTrackedTimer (Variables|GetPlayerToTrack) (select (Variables|UserOptions|GetbStartAtGameStartKnob) (Utilities|Enum|LiteralenumETimerDeviceState "Activated") (Utilities|Enum|LiteralenumETimerDeviceState "Enabled")))))

(fn StopTrackingPlayer
  (params (TrackedPlayer :type Controller%20Object%20Reference))
  (if (Utilities|Array|ContainsItem (CallFunction|GetTrackedPlayers) TrackedPlayer)
    (Utilities|Map|Remove (Variables|Default|GetStateBeforeDisabledCache) TrackedPlayer)
    (CallFunction|TimerClearHandles TrackedPlayer)
    (CallFunction|RemoveTrackedTimer TrackedPlayer)
    (CallFunction|ServerRequestClearDisplay (Utilities|Array|MakeArray TrackedPlayer))))

(fn TimerwasSuccessful
  (params (Timer%20Remaining :type Float%20%28double-precision%29))
  (return (or (Variables|UserOptions|GetbSuccessOnEndKnob) (> Timer%20Remaining 0.0))))

(fn UnbindDelegates
  (bind _fortminigameprogress (Variables|Components|GetFortMinigameProgress))
  (bind _exitedeventhandle (Variables|Default|GetExitedEventHandle))
  (Minigame|UnbindEventfromMinigameStateChangedDelegate _fortminigameprogress (EventDispatchers|CreateEvent :!event OnMinigameStateChanged))
  (Minigame|UnbindEventfromMinigameStartedDelegate _fortminigameprogress (EventDispatchers|CreateEvent :!event OnMinigameStarted))
  (Minigame|UnbindEventfromMinigameEndedDelegate _fortminigameprogress (EventDispatchers|CreateEvent :!event OnMinigameEnded))
  (Utilities|IsValid _exitedeventhandle
    (:Is%20Valid
      (GameplayEventRouter|StopListeningtoEvent _exitedeventhandle))))

(fn UpdateTimerColor
  (params (Color :type Linear%20Color%20Structure))
  (Rendering|Components|TextRender|SetTextRenderColor (Variables|Components|GetTimerDisplayTextFront) (Math|Conversions|ToColor(LinearColor) Color))
  (Class|CreativeGenericTextWidgetComp|UpdateColor (Variables|Default|GetTimerWidget) (Variables|GetColor)))

(fn UpdateDisplayStateTextAndFX
  (params (CurrentState :type ETimerDeviceState%20Enum :default "Enabled") (TimeRemaining :type Float%20%28double-precision%29) (bFreshTick :type Boolean))
  (Networking|CanExecuteCosmeticEvents
    (:True
      (Utilities|FlowControl|Sequence
        (:then_0
          (switch Utilities|FlowControl|Switch|SwitchonETimerDeviceState (Variables|GetCurrentState)
            (:Enabled
              (LocalVariables|SetTextToDisplay (Variables|UserOptions|GetTimerNotStartedTextKnob)))
            (:Activated
              (bind _output_get (LocalVariables|SetbInUrgency (and (> (Variables|UserOptions|GetUrgencyModeTimeKnob) 0.0) (<= (Variables|GetTimeRemaining) (Variables|UserOptions|GetUrgencyModeTimeKnob)))))
              (LocalVariables|SetTextToDisplay (Utilities|String|SelectString (Variables|UserOptions|GetUrgencyTextKnob) (Variables|UserOptions|GetTimerRunningTextKnob) (and _output_get (not (Utilities|String|IsEmpty (Variables|UserOptions|GetUrgencyTextKnob)))))))
            (:Paused
              (LocalVariables|SetTextToDisplay (Variables|UserOptions|GetTimerPausedTextKnob)))
            (:Completed
              ;[1 1 1] Assumed display text
              (if (CallFunction|TimerWasSuccessful (Variables|GetTimeRemaining))
                (LocalVariables|SetTextToDisplay "SUCCESS")
                (else
                  (LocalVariables|SetTextToDisplay "FAILED"))))
            (:Disabled
              (LocalVariables|SetTextToDisplay "DISABLED"))))
        (:then_1
          (if (Utilities|String|NotEqualExactly(String) (Variables|FortCreativeTimerDevice|GetClientCurrentSecondaryText) (LocalVariables|GetTextToDisplay))
            (bind _output_get_2 (Variables|FortCreativeTimerDevice|SetClientCurrentSecondaryText (LocalVariables|GetTextToDisplay)))
            (Class|CreativeGenericTextWidgetComp|UpdateText (Variables|Default|GetSecondaryWidget) (Utilities|Text|ToText(String) _output_get_2))))
        (:then_2
          (if (LocalVariables|GetbInUrgency)
            (if (Variables|GetbFreshTick)
              (CallFunction|TriggerUrgentFX))
            (else
              (Audio|Components|Audio|Stop (Variables|Default|GetUrgentBuildAudio))
              (Variables|Cosmetic|SetbUrgencyHasStarted false)
              (Visual|UpdateTimerColor (Variables|UserOptions|GetTimerColorKnob)))))))))

(fn UpdateDisplayTime
  (params (CurrentState :type ETimerDeviceState%20Enum :default "Enabled") (TimeRemaining :type Float%20%28double-precision%29))
  (bind _bcountdowndirectionknob (Variables|UserOptions|GetbCountDownDirectionKnob))
  (bind _returnvalue (Math|Float|Ceil TimeRemaining))
  (bind _durationknob (Variables|UserOptions|GetDurationKnob))
  (bind _returnvalue_2 (Math|Float|Truncate _durationknob))
  (bind _returnvalue_3 (- _returnvalue_2 _returnvalue))
  (bind _returnvalue_4 (Math|Integer|SelectInt _returnvalue _returnvalue_3 _bcountdowndirectionknob))
  (bind _currentstate (Variables|GetCurrentState))
  (bind _durationknob_2 (Variables|UserOptions|GetDurationKnob))
  (bind _returnvalue_5 (Math|Float|Truncate _durationknob_2))
  (bind _returnvalue_6 (Utilities|Select :Index _currentstate :Enabled _returnvalue_5 :Activated _returnvalue_4 :Paused _returnvalue_4 :Completed _returnvalue_4 :Disabled 0))
  (bind _returnvalue_7 (Math|Integer|Clamp(Integer) _returnvalue_6 0 _returnvalue_5))
  (bind (_outminutes _outseconds) (Utilities|ConvertSecondstoMinutesSeconds (Variables|FortCreativeTimerDevice|GetClientCurrentTime)))
  (if (!= (Variables|FortCreativeTimerDevice|GetClientCurrentTime) _returnvalue_7)
    (Variables|FortCreativeTimerDevice|SetClientCurrentTime _returnvalue_7)
    (Rendering|Components|TextRender|SetText (Variables|Components|GetTimerDisplayTextFront) (select (Variables|UserOptions|GetbDisplayTimeInMinutesSecondsKnob) (Utilities|Text|FormatText "NSLOCTEXT(\"[B33C6B864343CD24A4142E94EAC1B492]\", \"EB44032142EC6C3984966D9F1D342A27\", \"{Min}:{Sec}\")" (Utilities|Text|ToText(Integer) _outminutes false true 2) (Utilities|Text|ToText(Integer) _outseconds false true 2)) (Utilities|Text|ToText(Integer) (Variables|FortCreativeTimerDevice|GetClientCurrentTime))))
    (return true)
    (else
      (return false))))

(fn UpdateInGameDisplay
  (params (CurrentState :type ETimerDeviceState%20Enum :default "Enabled") (TimeRemaining :type Float%20%28double-precision%29))
  (bind _bfreshtick (Visual|UpdateDisplayTime CurrentState TimeRemaining))
  (bind _output_get (Variables|FortCreativeTimerDevice|SetClientCurrentState (Variables|GetCurrentState)))
  (Visual|UpdateDisplayStateTextandFX _output_get (Variables|GetTimeRemaining) _bfreshtick)
  (if (Math|Conversions|ToBoolean(Integer) (Variables|UserOptions|GetShowOnHUDKnob))
    (CallFunction|BroadcastUpdatetoHUD)))

(fn UpdateInGameVisibilities
  (Visibility|SetVisiblityDuringPhase (Variables|Components|GetVisibleInGameComponent) (select (Variables|UserOptions|GetbIsVisibleDuringGamesKnob) "NewEnumerator2" "NewEnumerator0"))
  (Visibility|SetVisiblityDuringPhase (Variables|Components|GetTextVisibileInGameComponent) (select (Variables|UserOptions|GetbIsVisibleDuringGamesKnob) "NewEnumerator3" "NewEnumerator1")))

(fn ConstructionScript
  (|CallParentFunction))
