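# Unlit masked material for the Creative timer device's text, drawn from a distance-field font with an adjustable outline.
# The font sample's red channel is remapped across a very wide symmetric range, steepening the distance field into a hard glyph edge.
# The same channel drives a SphereMask, with position, thickness and hardness parameters, that blends vertex color toward the Outline Color parameter.
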Material /Game/Creative/Devices/TimerDevice/Materials/M_CP_Device_Timer_Text
graph {
 properties {blendMode=BLEND_Masked, pixelDepthOffsetMode=PDOM_Legacy, shadingModel=MSM_Unlit, usedShadingModels=Unlit}
 nodes [
  n1 = FontSampleParameter() {font=/Engine/EngineFonts/RobotoDistanceField, parameterName=Font}
  LinearInterpolate(Alpha=@n1.R) {constA=-90, constB=90}
  LinearInterpolate(
   A=VertexColor()
   B=VectorParameter() {parameterName=Outline Color, defaultValue=<rgba 0 0 0 1>}.RGB
   Alpha=SphereMask(
    A=@n1.R
    B=ScalarParameter() {defaultValue=0.01, parameterName=OutlinePos}
    Radius=ScalarParameter() {defaultValue=0.5, parameterName=OutlineThickness}
    Hardness=ScalarParameter() {defaultValue=1, parameterName=OutlineHardness}
   )
  )
 ]
}
